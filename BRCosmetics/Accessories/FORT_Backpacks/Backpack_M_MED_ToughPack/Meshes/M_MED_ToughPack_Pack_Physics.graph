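# Physics asset for the ToughPack backpack cosmetic, giving physics bodies and joints to its figure-like limb chains and a belt-hung bottle.
# Small base_ capsules for head, hip, upper arms and bag belt attach to the pack body with free angular motion.
# The dyn_ arm, hip and belt-to-bottle chains hang from these anchors, and both leg chains branch from dyn_hip.
# Tight swing and twist limits, mostly near-zero twist, restrict how far the chains bend.
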PhysicsAsset /BRCosmetics/Accessories/FORT_Backpacks/Backpack_M_MED_ToughPack/Meshes/M_MED_ToughPack_Pack_Physics
bodies {
 base_bagbelt [{shapeType=Capsule, length=0.505, radius=0.505, shapeName=base_bagbelt_capsule, rotation=<rot 0 -90 0>}]
 base_head [{shapeType=Capsule, length=0.505, radius=0.505, shapeName=base_head_capsule, rotation=<rot 0 -90 0>}]
 base_hip [{shapeType=Capsule, length=0.505, radius=0.505, shapeName=base_hip_capsule, rotation=<rot 0 -90 0>}]
 base_upperarm_l [{shapeType=Capsule, length=0.505, radius=0.505, shapeName=base_upperarm_l_capsule, rotation=<rot 0 -90 0>}]
 base_upperarm_r [{shapeType=Capsule, length=0.505, radius=0.505, shapeName=base_upperarm_r_capsule, rotation=<rot 0 -90 0>}]
 base_upperleg_l [{shapeType=Capsule, length=0.505, radius=0.505, shapeName=base_upperleg_l_capsule, rotation=<rot 0 -90 0>}]
 base_upperleg_r [{shapeType=Capsule, length=0.505, radius=0.505, shapeName=base_upperleg_r_capsule, rotation=<rot 0 -90 0>}]
 dyn_bagbelt [{shapeType=Capsule, length=1, radius=1.2498, shapeName=dyn_bagbelt_capsule, center=<1.5 0 0>, rotation=<rot 90 0 0>}]
 dyn_bottle [
  {
   shapeType=Box
   extentX=5.1172
   extentY=2.7612
   extentZ=5.8155
   shapeName=dyn_bottle_box
   center=<1.6108 -0.2388 1.195>
   rotation=<rot -8.375 -90.2142 16.5048>
  }
  {
   shapeType=Box
   extentX=1.2008
   extentY=7.0945
   extentZ=1.5834
   shapeName=dyn_bottle_box
   center=<-1.4177 2.4038 0.7087>
   rotation=<rot -1.2814 97.4221 9.2225>
  }
 ]
 dyn_foot_l [{shapeType=Capsule, length=1, radius=1.4807, shapeName=dyn_foot_l_capsule, center=<2 0 0>, rotation=<rot 90 0 0>}]
 dyn_foot_r [{shapeType=Capsule, length=1, radius=1.4807, shapeName=dyn_foot_r_capsule, center=<-2 0 0>, rotation=<rot 90 0 0>}]
 dyn_hand_l [{shapeType=Capsule, length=1, radius=1.3654, shapeName=dyn_hand_l_capsule, center=<1.3783 0 0>, rotation=<rot 90 0 0>}]
 dyn_hand_r [{shapeType=Capsule, length=1, radius=1.3654, shapeName=dyn_hand_r_capsule, center=<-2 0 0>, rotation=<rot 90 180 0>}]
 dyn_hip [{shapeType=Capsule, length=2, radius=1.7498, shapeName=dyn_hip_capsule, center=<2 0 0>}]
 dyn_knee_l [{shapeType=Capsule, length=1, radius=1.4807, shapeName=dyn_knee_l_capsule, center=<2 0 0>, rotation=<rot 90 0 0>}]
 dyn_knee_r [{shapeType=Capsule, length=1, radius=1.4807, shapeName=dyn_knee_r_capsule, center=<-2 0 0>, rotation=<rot 90 0 0>}]
 dyn_lowerarm_l [{shapeType=Capsule, length=1, radius=1.3654, shapeName=dyn_lowerarm_l_capsule, center=<2 0 0>, rotation=<rot 90 0 0>}]
 dyn_lowerarm_r [{shapeType=Capsule, length=1, radius=1.3654, shapeName=dyn_lowerarm_r_capsule, center=<-2 0 0>, rotation=<rot 90 180 0>}]
 dyn_upperarm_l [{
   shapeType=Capsule
   length=6.2283
   radius=1.3346
   shapeName=dyn_upperarm_l_capsule
   center=<5.4116 0.0107 0.1959>
   rotation=<rot 81.738 -178.7486 179.9992>
  }]
 dyn_upperarm_r [{shapeType=Capsule, length=1, radius=1.2498, shapeName=dyn_upperarm_r_capsule, center=<-2 0 0.3893>, rotation=<rot 86.9308 179.9992 -0.0009>}]
 dyn_upperleg_l [{shapeType=Capsule, length=1, radius=1.4807, shapeName=dyn_upperleg_l_capsule, center=<2 0 0>, rotation=<rot 90 0 0>}]
 dyn_upperleg_r [{shapeType=Capsule, length=1, radius=1.4807, shapeName=dyn_upperleg_r_capsule, center=<-2 0 0>, rotation=<rot 90 0 0>}]
 pack [
  {
   shapeType=Capsule
   length=3.92
   radius=3.1404
   shapeName=pack_capsule
   center=<0.2262 -5.7861 -1.5476>
   rotation=<rot 1.6486 9.0656 -175.6048>
  }
  {shapeType=Capsule, length=15, radius=18.0597, shapeName=pack_capsule, center=<-3.809 15.869 -3.4917>, rotation=<rot 0 0 1.5708>}
 ]
}
constraints [
 {bone1Name=dyn_hip, bone2Name=base_hip, swing1LimitDegrees=70, swing2LimitDegrees=5, twistLimitDegrees=1}
 {
  bone1Name=base_upperleg_l
  bone2Name=dyn_hip
  swing1LimitDegrees=45
  swing1Motion=Free
  swing2LimitDegrees=45
  swing2Motion=Free
  twistLimitDegrees=45
  twistMotion=Free
 }
 {bone1Name=base_upperleg_r, bone2Name=dyn_hip, swing1LimitDegrees=45, swing2LimitDegrees=45, twistLimitDegrees=45}
 {bone1Name=dyn_upperleg_l, bone2Name=base_upperleg_l, swing1LimitDegrees=40, swing2LimitDegrees=1, twistLimitDegrees=1}
 {bone1Name=dyn_knee_l, bone2Name=dyn_upperleg_l, swing1LimitDegrees=10, swing2LimitDegrees=0, twistLimitDegrees=0}
 {bone1Name=dyn_foot_l, bone2Name=dyn_knee_l, swing1LimitDegrees=30, swing2LimitDegrees=0, twistLimitDegrees=0}
 {bone1Name=dyn_upperleg_r, bone2Name=base_upperleg_r, swing1LimitDegrees=40, swing2LimitDegrees=1, twistLimitDegrees=1}
 {bone1Name=dyn_knee_r, bone2Name=dyn_upperleg_r, swing1LimitDegrees=10, swing2LimitDegrees=0, twistLimitDegrees=0}
 {bone1Name=dyn_foot_r, bone2Name=dyn_knee_r, swing1LimitDegrees=30, swing2LimitDegrees=0, twistLimitDegrees=0}
 {bone1Name=dyn_upperarm_l, bone2Name=base_upperarm_l, swing1LimitDegrees=30, swing2LimitDegrees=25, twistLimitDegrees=5}
 {bone1Name=dyn_lowerarm_l, bone2Name=dyn_upperarm_l, swing1LimitDegrees=10, swing2LimitDegrees=11, twistLimitDegrees=0}
 {bone1Name=dyn_hand_l, bone2Name=dyn_lowerarm_l, swing1LimitDegrees=20, swing2LimitDegrees=20, twistLimitDegrees=0}
 {
  bone1Name=base_upperarm_l
  bone2Name=pack
  swing1LimitDegrees=45
  swing1Motion=Free
  swing2LimitDegrees=45
  swing2Motion=Free
  twistLimitDegrees=45
  twistMotion=Free
 }
 {
  bone1Name=base_hip
  bone2Name=pack
  swing1LimitDegrees=45
  swing1Motion=Free
  swing2LimitDegrees=45
  swing2Motion=Free
  twistLimitDegrees=45
  twistMotion=Free
 }
 {
  bone1Name=base_upperarm_r
  bone2Name=pack
  swing1LimitDegrees=45
  swing1Motion=Free
  swing2LimitDegrees=45
  swing2Motion=Free
  twistLimitDegrees=45
  twistMotion=Free
 }
 {
  bone1Name=base_head
  bone2Name=pack
  swing1LimitDegrees=45
  swing1Motion=Free
  swing2LimitDegrees=45
  swing2Motion=Free
  twistLimitDegrees=45
  twistMotion=Free
 }
 {bone1Name=dyn_upperarm_r, bone2Name=base_upperarm_r, swing1LimitDegrees=30, swing2LimitDegrees=25, twistLimitDegrees=5}
 {bone1Name=dyn_lowerarm_r, bone2Name=dyn_upperarm_r, swing1LimitDegrees=5, swing2LimitDegrees=1, twistLimitDegrees=0}
 {bone1Name=dyn_hand_r, bone2Name=dyn_lowerarm_r, swing1LimitDegrees=20, swing2LimitDegrees=20, twistLimitDegrees=0}
 {
  bone1Name=base_bagbelt
  bone2Name=pack
  swing1LimitDegrees=45
  swing1Motion=Free
  swing2LimitDegrees=45
  swing2Motion=Free
  twistLimitDegrees=45
  twistMotion=Free
 }
 {bone1Name=dyn_bagbelt, bone2Name=base_bagbelt, swing1LimitDegrees=5, swing2LimitDegrees=15, twistLimitDegrees=0}
 {bone1Name=dyn_bottle, bone2Name=dyn_bagbelt, swing1LimitDegrees=15, swing2LimitDegrees=10, twistLimitDegrees=1}
]
