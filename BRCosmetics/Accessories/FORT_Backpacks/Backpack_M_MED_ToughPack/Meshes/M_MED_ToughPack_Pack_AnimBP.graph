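# Animation Blueprint for the ToughPack backpack cosmetic, adding physics-driven secondary motion to its bones.
# Additive bone offsets, some gated by rigid-body and parachute flags, form a cached pose feeding parallel world-space and pack-relative RigidBody plus dyn_head AnimDynamics chains.
# A bool blend joins both chains before final left-arm offsets.
# Each update, a deprecated helper derives gravity and sim-space settings from per-movement-state parameters and stores them in variables.
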AnimBlueprint /BRCosmetics/Accessories/FORT_Backpacks/Backpack_M_MED_ToughPack/Meshes/M_MED_ToughPack_Pack_AnimBP
target_skeleton=&/M_MED_ToughPack_Pack_Skeleton
graphs [
 AnimGraph {nodes [
   n1 = RigidBody!(
    ComponentPose=Local To Component(LocalPose=Use cached pose 'CRCache_predyn'())
    Alpha
    OverrideWorldGravity:Vector="0,0,0"
    ExternalForce
    SimSpaceSettings:Sim Space Settings Structure=(WorldAlpha=1,VelocityScaleZ=0.3,DampingAlpha=0.3,MaxLinearVelocity=3000,MaxAngularVelocity=3000,MaxLinearAcceleration=3000,MaxAngularAcceleration=3006,ExternalLinearDragV=(X=0.2,Y=0.2,Z=0.2),ExternalLinearVelocity=(X=0,Y=0,Z=0),ExternalAngularVelocity=(X=0,Y=0,Z=0))
   ) -> (Pose) {
    bOverrideWorldGravity=true
    overridePhysicsAsset=&/M_MED_ToughPack_Pack_Physics
    simulationSpace=WorldSpace
    alphaBoolBlend {blendOption=Linear}
    baseBoneRef {boneName=pack}
    componentAppliedLinearAccClamp=<5000 5000 5000>
    simSpaceSettings {
     dampingAlpha=0.3
     maxAngularAcceleration=3006
     maxAngularVelocity=3000
     maxLinearAcceleration=3000
     maxLinearVelocity=3000
     velocityScaleZ=0.3
     worldAlpha=1
     externalLinearDragV=<0.2 0.2 0.2>
    }
   }
   n2 = Anim Dynamics - Bone: dyn_head!(ComponentPose=@n1, Alpha=1, GravityOverride:Vector="0,0,-980", ExternalForce="0,0,0") -> (Pose) {
    angularDampingOverride=1
    angularSpringConstant=20000
    bAngularSpring=true
    bOverrideAngularDamping=true
    bUseGravityOverride=true
    gravityScale=0.35
    simulationSpace=World
    alphaBoolBlend {blendOption=Linear}
    boundBone {boneName=dyn_head}
    componentLinearAccScale=<1 1 1>
    componentLinearVelScale=<1 1 1>
    gravityOverride {z=-980}
    physicsBodyDefinitions [{
      boundBone {boneName=dyn_head}
      boxExtents=<20 20 30>
      constraintSetup {angularLimitsMax=<-5 5 10>, angularLimitsMin=<-5 -5 -50>, angularTarget=<1 0 0>}
      localJointOffset=<3 0 0>
     }]
    relativeSpaceBone {boneName=pack}
   }
   n3 = Component To Local(ComponentPose=@n2)
   n4 = RigidBody(
    ComponentPose=Local To Component(LocalPose=Use cached pose 'CRCache_predyn'())
    OverrideWorldGravity="0,0,0"
    SimSpaceSettings=(WorldAlpha=1,VelocityScaleZ=0.3,DampingAlpha=0.3,MaxLinearVelocity=3000,MaxAngularVelocity=3000,MaxLinearAcceleration=3000,MaxAngularAcceleration=3006,ExternalLinearDragV=(X=0.2,Y=0.2,Z=0.2),ExternalLinearVelocity=(X=0,Y=0,Z=0),ExternalAngularVelocity=(X=0,Y=0,Z=0))
   ) {
    bOverrideWorldGravity=true
    overridePhysicsAsset=&M_MED_ToughPack_Pack_Physics
    simulationSpace=BaseBoneSpace
    alphaBoolBlend {blendOption=Linear}
    baseBoneRef {boneName=pack}
    componentAppliedLinearAccClamp=<5000 5000 5000>
    simSpaceSettings {
     dampingAlpha=0.3
     maxAngularAcceleration=3006
     maxAngularVelocity=3000
     maxLinearAcceleration=3000
     maxLinearVelocity=3000
     velocityScaleZ=0.3
     worldAlpha=1
     externalLinearDragV=<0.2 0.2 0.2>
    }
   }
   n5 = Anim Dynamics - Bone: dyn_head(ComponentPose=@n4, Alpha=1, GravityOverride="0,0,-980", ExternalForce="0,0,0") {
    angularDampingOverride=1
    angularSpringConstant=20000
    bAngularSpring=true
    bOverrideAngularDamping=true
    bUseGravityOverride=true
    gravityScale=0.35
    simulationSpace=BoneRelative
    alphaBoolBlend {blendOption=Linear}
    boundBone {boneName=dyn_head}
    componentLinearAccScale=<1 1 1>
    componentLinearVelScale=<1 1 1>
    gravityOverride {z=-980}
    physicsBodyDefinitions [{
      boundBone {boneName=dyn_head}
      boxExtents=<20 20 30>
      constraintSetup {angularLimitsMax=<-5 5 10>, angularLimitsMin=<-5 -5 -50>, angularTarget=<1 0 0>}
      localJointOffset=<3 0 0>
     }]
    relativeSpaceBone {boneName=pack}
   }
   n6 = Component To Local(ComponentPose=@n5)
   n7 = "Transform (Modify) Bone - Bone: dyn_upperarm_l"!(
    ComponentPose=Local To Component(LocalPose=Blend Poses by bool(BlendPose_0=@n3, BlendPose_1=@n6) {blendType=Linear})
    bAlphaBoolEnabled:bool=True
    Translation="0,0,0"
    Rotation="-5,0,0"
   ) -> (Pose) {
    alphaInputType=Bool
    rotationMode=BMM_Additive
    translationMode=BMM_Additive
    alphaBoolBlend {blendOption=Linear}
    boneToModify {boneName=dyn_upperarm_l}
    rotation {pitch=-5}
   }
   n8 = "Transform (Modify) Bone - Bone: dyn_upperarm_l"(ComponentPose=@n7, bAlphaBoolEnabled=True, Translation="0,0,0", Rotation="-1,0,0") {
    alphaInputType=Bool
    rotationMode=BMM_Additive
    translationMode=BMM_Additive
    alphaBoolBlend {blendOption=Linear}
    boneToModify {boneName=dyn_upperarm_l}
    rotation {pitch=-1}
   }
   Output Pose(Result=Component To Local(ComponentPose=@n8)) {layerGroup=None}
   n9 = "Transform (Modify) Bone - Bone: dyn_head"!(
    ComponentPose=Local To Component(LocalPose=Input Pose())
    bAlphaBoolEnabled:bool=True
    Translation="0,0,0"
    Rotation="0,0,-10"
    Scale="1,1,1"
   ) -> (Pose) {
    alphaInputType=Bool
    rotationMode=BMM_Additive
    translationMode=BMM_Additive
    alphaBoolBlend {blendOption=Linear}
    boneToModify {boneName=dyn_head}
    rotation {roll=-10}
   }
   n10 = Get bRigidBodyEnabled()
   n11 = "Transform (Modify) Bone - Bone: dyn_upperleg_l"!(ComponentPose=@n9, bAlphaBoolEnabled:bool=@n10, Translation="0,0,0", Rotation="30,0,0", Scale="1,1,1") -> (Pose) {
    alphaInputType=Bool
    rotationMode=BMM_Additive
    translationMode=BMM_Additive
    alphaBoolBlend {blendOption=Linear}
    boneToModify {boneName=dyn_upperleg_l}
    rotation {pitch=30}
   }
   n12 = Get bIsParachuteOpened()
   n13 = "Transform (Modify) Bone - Bone: base_upperarm_l"!(ComponentPose=@n11, bAlphaBoolEnabled:bool=@n12, Translation="0,0,0", Rotation="-20,0,0", Scale="1,1,1") -> (Pose) {
    alphaInputType=Bool
    rotationMode=BMM_Additive
    rotationSpace=BCS_ParentBoneSpace
    translationMode=BMM_Additive
    translationSpace=BCS_ParentBoneSpace
    alphaBoolBlend {blendOption=Linear}
    boneToModify {boneName=base_upperarm_l}
    rotation {pitch=-20}
   }
   n14 = "Transform (Modify) Bone - Bone: base_bagbelt"!(ComponentPose=@n13, bAlphaBoolEnabled:bool=True, Translation="0,0,0", Rotation="0,5,0") -> (Pose) {
    alphaInputType=Bool
    rotationMode=BMM_Additive
    translationMode=BMM_Additive
    alphaBoolBlend {blendOption=Linear}
    boneToModify {boneName=base_bagbelt}
    rotation {yaw=5}
   }
   Save cached pose 'CRCache_predyn'(Pose=Component To Local(ComponentPose=@n14)) {cacheName=CRCache_predyn}
   AND Boolean(A=@n10, B=@n12)
  ]}
 EventGraph {nodes [
   Event BlueprintUpdateAnimation()
   > n1 = Gravity Override State Function DEPRECATED(
    CharacterPartAnimInstance=Self-Reference()
    GameTimeInSeconds=GetGameTimeInSeconds()
    GravityOverrideParams=Get Gravity Override Params()
    RBAN Params=Get RBAN Params()
   )
   > Set GravityOverride(GravityOverride=@n1)
   > Set Sim Space Settings(Sim Space Settings:Sim Space Settings Structure=Break RBANSetting Struct!(RBANSettingStruct:"RBANSetting Struct Structure (by ref)"=@n1.RBAN SettingsOut) -> (
     ComponentLinearAccScale_28_#1:Vector=~
     ComponentLinearVelScale_31_#2:Vector=~
     SimSpaceSettings_34_#3:Sim Space Settings Structure
    ) {is=title K2Node_BreakStruct}.SimSpaceSettings_34_#3) -> (Output_Get:Sim Space Settings Structure)
   TryGetPawnOwner()
  ]}
]
variables [
 Gravity Override Params {type=GravityOverrideParamsStruct}
 GravityOverride {type=Vector}
 RBAN Params {
  type=RBANParamsStruct
  default {
   dBNO {simSpaceSettings {
     dampingAlpha=0.3
     maxAngularAcceleration=3006
     maxAngularVelocity=3000
     maxLinearAcceleration=3000
     maxLinearVelocity=3000
     velocityScaleZ=0.3
     externalLinearDragV=<0.2 0.2 0.2>
    }}
   dBNO_Carried {simSpaceSettings {
     dampingAlpha=0.3
     maxAngularAcceleration=3006
     maxAngularVelocity=3000
     maxLinearAcceleration=3000
     maxLinearVelocity=3000
     velocityScaleZ=0.3
     externalLinearDragV=<0.2 0.2 0.2>
    }}
   emoteOrMelee {simSpaceSettings {
     dampingAlpha=0.3
     maxAngularAcceleration=3006
     maxAngularVelocity=3000
     maxLinearAcceleration=3000
     maxLinearVelocity=3000
     velocityScaleZ=0.3
     externalLinearDragV=<0.2 0.2 0.2>
    }}
   front_End {simSpaceSettings {
     dampingAlpha=0.5
     maxAngularAcceleration=1500
     maxAngularVelocity=1500
     maxLinearAcceleration=1500
     maxLinearVelocity=1500
     velocityScaleZ=0.5
     externalLinearDragV=<0.2 0.2 0.2>
    }}
   onGround_Crouch_Moving {simSpaceSettings {
     dampingAlpha=0.3
     maxAngularAcceleration=3006
     maxAngularVelocity=3000
     maxLinearAcceleration=3000
     maxLinearVelocity=3000
     velocityScaleZ=0.3
     externalLinearDragV=<0.2 0.2 0.2>
    }}
   onGround_Falling {simSpaceSettings {
     dampingAlpha=0.3
     maxAngularAcceleration=3006
     maxAngularVelocity=3000
     maxLinearAcceleration=3000
     maxLinearVelocity=3000
     velocityScaleZ=0.3
     externalLinearDragV=<0.2 0.2 0.2>
    }}
   onGround_Moving {simSpaceSettings {
     dampingAlpha=0.2
     maxAngularAcceleration=5000
     maxAngularVelocity=5000
     maxLinearAcceleration=5000
     maxLinearVelocity=5000
     velocityScaleZ=0.2
     externalLinearDragV=<0.2 0.2 0.2>
    }}
   onGround_Standing {simSpaceSettings {
     dampingAlpha=0.3
     maxAngularAcceleration=3006
     maxAngularVelocity=3000
     maxLinearAcceleration=3000
     maxLinearVelocity=3000
     velocityScaleZ=0.3
     externalLinearDragV=<0.2 0.2 0.2>
    }}
   rail_Grinding {simSpaceSettings {
     dampingAlpha=0.3
     maxAngularAcceleration=3006
     maxAngularVelocity=3000
     maxLinearAcceleration=3000
     maxLinearVelocity=3000
     velocityScaleZ=0.3
     worldAlpha=1
     externalLinearDragV=<0.2 0.2 0.2>
    }}
   riding_Moving {simSpaceSettings {
     dampingAlpha=0.3
     maxAngularAcceleration=3006
     maxAngularVelocity=3000
     maxLinearAcceleration=3000
     maxLinearVelocity=3000
     velocityScaleZ=0.3
     externalLinearDragV=<0.2 0.2 0.2>
    }}
   riding_Moving_Targeting {simSpaceSettings {
     dampingAlpha=0.3
     maxAngularAcceleration=3006
     maxAngularVelocity=3000
     maxLinearAcceleration=3000
     maxLinearVelocity=3000
     velocityScaleZ=0.3
     externalLinearDragV=<0.2 0.2 0.2>
    }}
   riding_Standing {simSpaceSettings {
     dampingAlpha=0.3
     maxAngularAcceleration=3006
     maxAngularVelocity=3000
     maxLinearAcceleration=3000
     maxLinearVelocity=3000
     velocityScaleZ=0.3
     externalLinearDragV=<0.2 0.2 0.2>
    }}
   skydive_DivingDown {simSpaceSettings {
     dampingAlpha=0.3
     maxAngularAcceleration=3006
     maxAngularVelocity=3000
     maxLinearAcceleration=3000
     maxLinearVelocity=3000
     velocityScaleZ=0.3
     externalLinearDragV=<0.2 0.2 0.2>
    }}
   skydive_FloatingDown {
    componentLinearAccScale=<0 0.5 0>
    componentLinearVelScale=<0 0.5 0>
    simSpaceSettings {
     dampingAlpha=0.3
     maxAngularAcceleration=3006
     maxAngularVelocity=3000
     maxLinearAcceleration=3000
     maxLinearVelocity=3000
     velocityScaleZ=0.3
     externalLinearDragV=<0.2 0.2 0.2>
    }
   }
   skydive_GoingUp {simSpaceSettings {
     dampingAlpha=0.3
     maxAngularAcceleration=3006
     maxAngularVelocity=3000
     maxLinearAcceleration=3000
     maxLinearVelocity=3000
     velocityScaleZ=0.3
     externalLinearDragV=<0.2 0.2 0.2>
    }}
   skydive_Parachuting {
    componentLinearAccScale=<0 0.1 0.3>
    componentLinearVelScale=<0 0.1 0.3>
    simSpaceSettings {
     dampingAlpha=0.3
     maxAngularAcceleration=3006
     maxAngularVelocity=3000
     maxLinearAcceleration=3000
     maxLinearVelocity=3000
     velocityScaleZ=0.3
     externalLinearDragV=<0.2 0.2 0.2>
    }
   }
   skydive_Surfing {simSpaceSettings {
     dampingAlpha=0.3
     maxAngularAcceleration=3006
     maxAngularVelocity=3000
     maxLinearAcceleration=3000
     maxLinearVelocity=3000
     velocityScaleZ=0.3
     externalLinearDragV=<0.2 0.2 0.2>
    }}
   swimming {simSpaceSettings {
     dampingAlpha=0.3
     maxAngularAcceleration=3006
     maxAngularVelocity=3000
     maxLinearAcceleration=3000
     maxLinearVelocity=3000
     velocityScaleZ=0.3
     externalLinearDragV=<0.2 0.2 0.2>
    }}
   tactical_Sprint {simSpaceSettings {
     dampingAlpha=0.3
     maxAngularAcceleration=3006
     maxAngularVelocity=3000
     maxLinearAcceleration=3000
     maxLinearVelocity=3000
     velocityScaleZ=0.3
     externalLinearDragV=<0.2 0.2 0.2>
    }}
   ziplining {simSpaceSettings {
     dampingAlpha=0.3
     maxAngularAcceleration=3006
     maxAngularVelocity=3000
     maxLinearAcceleration=3000
     maxLinearVelocity=3000
     velocityScaleZ=0.3
     externalLinearDragV=<0.2 0.2 0.2>
    }}
  }
 }
 Sim Space Settings {
  type=SimSpaceSettings
  default {
   dampingAlpha=1
   maxAngularAcceleration=10000
   maxAngularVelocity=10000
   maxLinearAcceleration=10000
   maxLinearVelocity=10000
   velocityScaleZ=1
   worldAlpha=0
   externalLinearDragV=<0 0 0>
  }
 }
]
